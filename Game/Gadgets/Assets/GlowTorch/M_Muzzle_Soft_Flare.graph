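# Additive unlit soft flare material for particle and Niagara sprites, apparently the GlowTorch muzzle glow.
# Combines a sun glare texture with a thin horizontal SphereMask streak, both flickered by noise panned at zero UV tiling, so it varies only over time.
# The result is tinted, depth-faded and passed through an emissive adjustment function.
# The CameraFacing switch selects Color parameter tint with camera-aligning vertex offset, or particle color without offset.
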Material /Game/Gadgets/Assets/GlowTorch/M_Muzzle_Soft_Flare
graph {
 properties {
  bUsedWithNiagaraSprites=true
  bUsedWithParticleSprites=true
  blendMode=BLEND_Additive
  pixelDepthOffsetMode=PDOM_Legacy
  shadingModel=MSM_Unlit
  usedShadingModels=Unlit
 }
 nodes [
  n1 = StaticSwitchParameter(True=VectorParameter() {parameterName=Color, defaultValue=<rgba 1 1 1 0>}.RGB, False=ParticleColor().RGB) {parameterName=CameraFacing}
  n2 = TextureSample(UVs=Panner(Coordinate=TextureCoordinate() {uTiling=0, vTiling=0}) {speedX=0.31, speedY=0.108}) {texture=&T_NoisePack_01}
  n3 = LinearInterpolate(Alpha=@n2.R) {constA=0.5}
  n4 = Power(
   Base=SphereMask(A=TextureCoordinate() {vTiling=9}, B=Constant2Vector() {g=4.5, r=0.5}, Radius=Constant() {r=0.5}, Hardness=Constant() {r=0.05})
   Exp=LinearInterpolate(Alpha=@n2.R) {constA=8, constB=4}
  ) {constExponent=8}
  n5 = Add(
   A=Multiply(A=TextureSample() {texture=/Game/Effects/Fort_Effects/Textures/T_SUN_Glare_A}.G, B=@n3)
   B=Multiply(A=Multiply(A=@n4, B=@n3) {constB=0.3}) {constB=0.6}
  )
  MaterialFunctionCall(Emissive=DepthFade(Opacity=Multiply(A=@n1, B=@n5)) {fadeDistanceDefault=30}) {materialFunction=&MF_FX_EmissiveAdjustments}
  StaticSwitchParameter(
   True=MaterialFunctionCall!(
    WorldPosition=WorldPosition().XYZ
    PivotLocation
    WS Normals
    Custom Object Basis 1=Constant3Vector() {constant=<rgba 0 1 0 0>}
    Custom Object Basis 2=Constant3Vector() {constant=<rgba 1 0 0 0>}
    Custom Object Basis 3=Constant3Vector() {constant=<rgba 0 0 1 0>}
   ) {materialFunction=/Engine/Functions/Engine_MaterialFunctions02/WorldPositionOffset/AlignMeshToTheCamera}.World Position Offset
   False=Constant()
  ) {parameterName=CameraFacing}
 ]
}
