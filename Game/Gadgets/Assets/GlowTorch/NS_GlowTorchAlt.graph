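# Niagara system for an alternate glow torch effect, showing a soft, persistent sprite flare tinted by the User.Color parameter.
# Its single SoftFlair emitter spawns particles in an instantaneous burst as fixed-size camera-offset sprites.
# Particles loop their lifetime instead of dying when it elapses, with infinitely lived ones allowed to die when the emitter deactivates.
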NiagaraSystem /Game/Gadgets/Assets/GlowTorch/NS_GlowTorchAlt
dependencies {usedDataInterfaces [], usedDynamicInputs [], usedModules [&SystemState]}
emitters {SoftFlair {
  emitterUpdateScript [
   &EmitterState(
    Inactive Response=::NewEnumerator1
    Life Cycle Mode=::NewEnumerator1
    Loop Behavior=::NewEnumerator1
    Loop Duration Mode=::NewEnumerator1
    UseLoopDelay=false
   )
   &SpawnBurst_Instantaneous(Use Spawn Probability=false, UseLoopCountLimit=false)
  ]
  particleSpawnScript [&InitializeParticle(Color=$User.Color, Sprite Size=<150 150>, Sprite Size Mode=::NewEnumerator4, UsePositionOffset=false, Write Material Random=true)]
  particleUpdateScript [
   &ParticleState(
    Kill Particles When Lifetime Has Elapsed=false
    Let Infinitely Lived Particles Die When Emitter Deactivates=true
    Loop Particles Lifetime=true
   )
   &CameraOffset(Camera Offset Amount=10)
  ]
  rendererClasses [&NiagaraSpriteRendererProperties]
  renderers [{rendererClass=&NiagaraSpriteRendererProperties}]
 }}
summary {
 systemName=NS_GlowTorchAlt
 emitters [{emitterName=SoftFlair, rendererClasses [&NiagaraSpriteRendererProperties]}]
 userVariables [User.Color {type=LinearColor, description="", defaultValue {struct=LinearColor, value=<rgba 1 1 1 1>}}]
}
